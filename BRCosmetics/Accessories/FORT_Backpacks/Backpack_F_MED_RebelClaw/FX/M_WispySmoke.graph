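# Translucent, two-sided unlit material for wispy smoke on Niagara mesh particles in the RebelClaw backpack cosmetic.
# Color is a two-tone V gradient tinted by particle color.
# Opacity combines panning masks eroded by scrolling noise, with Niagara-driven dissolve, UV fade-in and depth fade outside mobile.
# A vertex offset bends the mesh toward a Niagara-supplied position, clamped and weighted along V, adding a sine wave along normals.
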Material /BRCosmetics/Accessories/FORT_Backpacks/Backpack_F_MED_RebelClaw/FX/M_WispySmoke
graph {
 properties {
  bUsedWithNiagaraMeshParticles=true
  blendMode=BLEND_Translucent
  pixelDepthOffsetMode=PDOM_Legacy
  shadingModel=MSM_Unlit
  twoSided=true
  usedShadingModels=Unlit
 }
 nodes [
  n1 = VectorParameter() {parameterName=Color_1, defaultValue=<rgba 0.3958 0.3804 0.3229 1>}
  n2 = VectorParameter() {parameterName=Color_2, defaultValue=<rgba 0.0934 0.0938 0.0658 1>}
  n3 = TextureCoordinate()
  n4 = Saturate(None=Power(Base=ComponentMask(None=@n3) {g=true}) {constExponent=0.5})
  MaterialFunctionCall(Emissive=Multiply(A=ParticleColor().RGB, B=LinearInterpolate(A=@n1.RGB, B=@n2.RGB, Alpha=@n4))) {materialFunction=&MF_FX_EmissiveAdjustments}
  n5 = ScalarParameter() {defaultValue=3, group=Alpha1, parameterName=Alpha1_Power, sortPriority=5}
  n6 = Multiply(A=Time(), B=ScalarParameter() {defaultValue=0.5, parameterName=Speed})
  n7 = TextureSample(UVs=Panner(Coordinate=TextureCoordinate() {uTiling=0.1, vTiling=0.5}, Time=@n6) {speedX=0.1}) {texture=/Game/Effects/Fort_Effects/Textures/T_Smoke_Ring_Mask2}
  n8 = VectorParameter() {parameterName=UV Controls, channelNames {a=Speed V, b=Speed U, g=Tile V, r=Tile U}, defaultValue=<rgba 1.5 1 -1 0>}
  n9 = MaterialFunctionCall(UVs=TextureCoordinate(), Texture Scale=AppendVector(A=@n8.Tile U, B=@n8.Tile U)) {materialFunction=&ScaleUVsByCenter}
  n10 = Constant() {r=0.25}
  n11 = TextureSampleParameter2D(UVs=Panner(
    Coordinate=MaterialFunctionCall(UVs=@n9.UVs, "Rotation Angle (0-1)"=@n10) {materialFunction=&CustomRotator}.Rotated Values
    Time=@n6
    Speed=AppendVector(A=@n8.Speed U, B=@n8.Speed U)
   ) {speedX=-1}) {parameterName=Main Tex, samplerType=SAMPLERTYPE_Grayscale, texture=&T_Smoke_Trail_Soft}
  n12 = ScalarParameter() {defaultValue=0.95, group=Disolve, parameterName=Disolve_U_Scale, sortPriority=1}
  n13 = ScalarParameter() {defaultValue=0.8, group=Disolve, parameterName=Disolve_V_Scale, sortPriority=1}
  n14 = Panner(
   Coordinate=Multiply(A=TextureCoordinate() {uTiling=0.5, vTiling=0.5}, B=AppendVector(A=@n12, B=@n13))
   Time=@n6
   Speed=AppendVector(
    A=ScalarParameter() {group=Disolve, parameterName=Disolve_U_Speed, sortPriority=3}
    B=ScalarParameter() {defaultValue=0.6, group=Disolve, parameterName=Disolve_V_Speed, sortPriority=3}
   )
  )
  n15 = Subtract(
   A=Multiply(A=Power(Base=@n7.R, Exp=@n5), B=Power(Base=@n11.G))
   B=LinearInterpolate(
    B=ScalarParameter() {defaultValue=0.1, group=Disolve, parameterName=DisolveHigh}
    Alpha=TextureSample(UVs=@n14) {texture=/Game/Effects/Fort_Effects/Textures/Tile/T_Inky_Smoke_Tile}.R
   )
  )
  n16 = DynamicParameter() {defaultValue=<rgba 0 0 0 1>, paramNames [TopFadeOutBlend, FadeIn, Dissolve, Param4]}
  n17 = Add(A=OneMinus(None=ScalarParameter() {parameterName=Erosion Contrast}), B=@n16.Dissolve)
  n18 = ComponentMask(None=TextureCoordinate()) {g=true}
  n19 = DynamicParameter() {defaultValue=<rgba 0 0 0 1>, paramNames [TopFadeOutBlend, FadeIn, Dissolve, Param4]}
  n20 = Multiply(
   A=Saturate(None=MaterialFunctionCall(Alpha=@n15, Min=@n16.Dissolve, Max=@n17) {materialFunction=&SmoothStep}.Result)
   B=Saturate(None=Subtract(A=@n18, B=@n19.FadeIn) {constB=0})
  )
  n21 = Multiply(A=@n20, B=DepthFade() {fadeDistanceDefault=1, opacityDefault=25})
  Multiply(A=ParticleColor().A, B=ShadingPathSwitch(Default=@n21, Deferred=@n21, Mobile=@n20) {desc=Migrated from FeatureLevelSwitch})
  n22 = DynamicParameter() {desc=Parameters set by Niagara, parameterIndex=1, paramNames [X,Y,Z,WPOIntensity]}
  n23 = ParticlePositionWS()
  n24 = DynamicParameter() {desc=Parameters set by Niagara, parameterIndex=2, defaultValue=<rgba 0 0 20 1>, paramNames [X Offset FE, Y Offset FE, Z Offset FE, Param4]}
  n25 = Transform(None=MaterialFunctionCall(X=@n24.X Offset FE, Y=@n24.X Offset FE, Z=@n24.X Offset FE) {materialFunction=&MakeFloat3}.Result) {transformSourceType=TRANSFORMSOURCE_World, transformType=TRANSFORM_Local}
  n26 = Clamp(
   None=Subtract(
    A=MaterialFunctionCall(X=@n22.X, Y=@n22.X, Z=@n22.X) {materialFunction=&MakeFloat3}.Result
    B=Add(A=@n23, B=@n25)
   )
   Min=ScalarParameter() {defaultValue=-100, group=WPO, parameterName=Clamp_min}
   Max=ScalarParameter() {defaultValue=100, group=WPO, parameterName=Clamp_max}
  ) {maxDefault=1000, minDefault=-1000}
  n27 = OneMinus(None=TextureCoordinate())
  n28 = Multiply(
   A=Power(Base=ComponentMask(None=@n27) {g=true}, Exp=ScalarParameter() {defaultValue=3, group=WPO, parameterName=WPOBendPower})
   B=@n22.WPOIntensity
  )
  n29 = Constant() {r=1}
  n30 = ScalarParameter() {defaultValue=0.5, parameterName=WaveTilingV}
  n31 = ComponentMask(None=Multiply(A=TextureCoordinate(), B=AppendVector(A=@n29, B=@n30))) {g=true}
  n32 = Power(Base=@n31)
  n33 = Add(A=@n31, B=Multiply(A=Time(), B=ScalarParameter() {defaultValue=2, parameterName=Wavespeed}) {constB=2})
  n34 = Multiply(A=Multiply(A=VertexNormalWS(), B=ScalarParameter() {defaultValue=2, parameterName=WaveStrenght}), B=Sine(None=Add(A=@n32, B=@n33)))
  Reroute(None=Add(A=LinearInterpolate(B=@n26, Alpha=@n28), B=Multiply(A=@n28, B=@n34)))
 ]
}
