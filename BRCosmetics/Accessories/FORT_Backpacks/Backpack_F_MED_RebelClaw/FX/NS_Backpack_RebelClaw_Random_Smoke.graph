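# Niagara system for the RebelClaw backpack cosmetic that puffs smoke sprites in short bursts separated by random delays, suppressed in the front end.
# One emitter spawns by rate and distance, partitioning particles between two mirrored offsets on the skeletal mesh.
# Smoothed owner speed blends drift, follow lag and curl noise strength, so the trail reacts to movement.
# Time-of-day tint, user start/end colors and animated material parameters shape each puff.
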NiagaraSystem /BRCosmetics/Accessories/FORT_Backpacks/Backpack_F_MED_RebelClaw/FX/NS_Backpack_RebelClaw_Random_Smoke
dependencies {
 usedDataInterfaces [&NiagaraDataInterfaceCurve, &NiagaraDataInterfaceSkeletalMesh]
 usedDynamicInputs []
 usedModules [
  &SystemState
  &self:SystemSpawnScript.SystemScriptSource.SystemScriptGraph.NiagaraNodeAssignment_0.SetVariables_#1
 ]
}
emitters {SmokePuff_Gen5 {
  emitterUpdateScript [
   &EmitterState(
    Life Cycle Mode=::NewEnumerator1
    Loop Delay=&RandomRangeFloat(Minimum=7, Maximum=10)
    Loop Duration=1.1
    Recalculate Duration Each Loop=true
    UseLoopDelay=true
   )
   &SpawnPerUnit(
    Module.Use Max Movement Threshold=true
    Module.Use Movement Tolerance=false
    Spawn Probability:NiagaraFloat=&Multiply_Float(A=&FN_FrontEndFloatModifier(Gameplay=0.5, Front End=0), B=$User.SpawnScale)
    Use Spawn Probability=true
   )
   &SpawnRate(
    Spawn Probability=&Multiply_Float(A=&FN_FrontEndFloatModifier(Gameplay=1, Front End=0), B=$User.SpawnScale)
    SpawnRate=8
    Use Spawn Probability=true
   )
  ]
  particleSpawnScript [
   &PartitionParticles()
   &InitializeParticle(
    Color=&FNTimeOfDayColor(Night Color=<rgba 1.25 1.25 1.25 0.5>, Morning Color=<rgba 1 0.5 1 0.5>, Day Color=<rgba 1 1 1 0.5>, Evening Color=<rgba 1.25 1.25 1.25 0.5>)
    Lifetime=&RandomRangeFloat(Minimum=0.8, Maximum=1)
    Sprite Rotation Mode=::NewEnumerator2
    Sprite Size Mode=::NewEnumerator3
    Uniform Sprite Size=30
    UsePositionOffset=false
    Write Material Random=true
   ) {module=InitializeParticle002}
   &CameraOffset(Camera Offset Amount=12)
  ]
  particleUpdateScript [
   &ParticleState()
   /SetVariables(Particles.SpeedScale:NiagaraFloat=&VectorLength(Vector=&Multiply_VectorByFloat(
      Vector=$Engine.Owner.Velocity
      Float=&RemapRange(Input Value To Remap=1, Input Min=0, Input Max=10000, Output Min=0, Output Max=1)
     ))) {isSetParameters=true}
   /SetVariables(Particles.SpeedScaleRemap:NiagaraFloat=&RemapRange(Input Value To Remap=$Particles.SpeedScale, Input Min=0, Input Max=0.25, Output Min=0, Output Max=1)) {isSetParameters=true}
   &InterpolateOverTime(Target Float Value:NiagaraFloat=$Particles.SpeedScaleRemap) {module=InterpolateOverTime001}
   &InterpolateOverTime(
    Rate Of Change:NiagaraFloat=&Lerp_Float(A=$User.Interpolate A, B=$User.Interpolate B, Alpha=$Particles.InterpolateOverTime001.InterpolateOverTime_Float.Moving Average)
    Target Vector Value:Vector3f=&TransformPosition(
     Position=&AddVectorToPosition(
      Position=$Particles.Position
      Vector=&TransformVector(
       Vector=&MakeVector(
        X=0
        Y=&FloatFromCurve(
         FloatCurve=DataInterface() {propertyValues {Curve {keys [<at 0.5 1 RCIM_Cubic>,<at 1 0 RCIM_Cubic>]}}}
         Scale Curve=&Lerp_Float(A=-10, B=-5, Alpha=$Particles.InterpolateOverTime001.InterpolateOverTime_Float.Moving Average)
        )
        Z=&FloatFromCurve(
         FloatCurve=DataInterface() {propertyValues {Curve {keys [
             {arriveTangent=1.6588, interpMode=RCIM_Cubic, leaveTangent=1.6588, tangentMode=RCTM_User, time=0.3, value=0.125}
             {arriveTangent=1.8455, interpMode=RCIM_Cubic, leaveTangent=1.8455, tangentMode=RCTM_User, time=0.6, value=0.925}
            ]}}}
         Scale Curve=&Lerp_Float(A=6, B=1, Alpha=$Particles.InterpolateOverTime001.InterpolateOverTime_Float.Moving Average)
        )
       )
       Destination Space=::Simulation
      )
     )
     Source Space=::Simulation
     Destination Space=::Local
    )
    Variable Type=::NewEnumerator2
   )
   &SkeletalMeshLocation(
    Bone Sampling Mode=::NewEnumerator4
    MaskSpawnGroup=false
    Sampled Position Offset=&MaskVectorBySpawnGroup(
     Masked Vector=&Lerp_Vector(
      A=<5 -12 15>
      B=$Particles.InterpolateOverTime.InterpolateOverTime_V3_.Moving Average
      Alpha=&FloatFromCurve(FloatCurve=DataInterface() {propertyValues {Curve {keys [{tangentMode=RCTM_User, time=0, value=0}, {tangentMode=RCTM_User, time=1, value=1}]}}})
     )
     Passthrough Vector=&Lerp_Vector(
      A=<-5 -12 15>
      B=$Particles.InterpolateOverTime.InterpolateOverTime_V3_.Moving Average
      Alpha=&FloatFromCurve(FloatCurve=DataInterface() {propertyValues {Curve {keys [{tangentMode=RCTM_User, time=0, value=0}, {tangentMode=RCTM_User, time=1, value=1}]}}})
     )
     Spawn Group Mask=$Particles.PartitionParticles.Partition
    )
    Skeletal Mesh=$System.Skeletal Mesh
   )
   &ScaleSpriteSize(Scale Factor=&Vector2DFromFloat(Value=&FloatFromCurve(
      FloatCurve=DataInterface() {propertyValues {Curve {keys [
          {arriveTangent=-0.0943, interpMode=RCIM_Cubic, leaveTangent=-0.0943, tangentMode=RCTM_User, time=0, value=0.3}
          {arriveTangent=0.116, interpMode=RCIM_Cubic, leaveTangent=0.116, tangentMode=RCTM_User, time=1, value=1.25}
         ]}}}
      Scale Curve=0.8
     )))
   &ScaleColor(
    Scale Alpha=&FloatFromCurve(
     FloatCurve=DataInterface() {propertyValues {Curve {keys [<at 0 0 RCIM_Cubic>,<at 0.2 0.5 RCIM_Cubic>,<at 1 0 RCIM_Cubic>]}}}
     Scale Curve=&Multiply_Float(
      A=3
      B=&FloatFromCurve(FloatCurve=DataInterface() {propertyValues {Curve {keys [
           <at 0 0 RCIM_Cubic>
           {arriveTangent=0.5, leaveTangent=0.5, time=0.5, value=1}
           {arriveTangent=-0.4, leaveTangent=-0.4, time=2, value=1}
           {interpMode=RCIM_Cubic, tangentMode=RCTM_User, time=3, value=0}
          ]}}})
     )
    )
    Scale RGB=&MakeVectorFromLinearColorRGB(Color=&LerpLinearColors(
      StartColor=$User.StartColor
      EndColor=$User.EndColor
      LerpFactor=&FloatFromCurve(FloatCurve=DataInterface() {propertyValues {Curve {keys [
           {arriveTangent=-0.1505, interpMode=RCIM_Cubic, leaveTangent=-0.1505, tangentMode=RCTM_User, time=0, value=0}
           {arriveTangent=-0.1542, interpMode=RCIM_Cubic, leaveTangent=-0.1542, tangentMode=RCTM_User, time=0.7, value=1}
          ]}}})
     ))
    ScaleA=true
    ScaleRGB=true
   ) {module=ScaleColor002}
   &DynamicMaterialParameters(
    Index 0 Param 1=&FloatFromCurve(
     FloatCurve=DataInterface() {propertyValues {Curve {keys [
         <at 0 0 RCIM_Cubic>
         {arriveTangent=1.5947, interpMode=RCIM_Cubic, leaveTangent=1.5947, tangentMode=RCTM_User, time=0.2557, value=0.5257}
         <at 1 1 RCIM_Cubic>
        ]}}}
     CurveIndex=$Particles.Age
     Scale Curve=0.25
    )
    Index 1 Param 1=1
    Index 1 Param 3=1
    Index 1 Param 4=0.5
    Index 2 Param 1=&FloatFromCurve(FloatCurve=DataInterface() {propertyValues {Curve {keys [
         {arriveTangent=1.9151, interpMode=RCIM_Cubic, leaveTangent=1.9151, tangentMode=RCTM_User, time=0, value=0}
         {interpMode=RCIM_Cubic, tangentMode=RCTM_User, time=1, value=1}
        ]}}})
    Index 2 Param 2=&FloatFromCurve(FloatCurve=DataInterface() {propertyValues {Curve {keys [<at 0 0.25 RCIM_Cubic>,<at 1 0 RCIM_Cubic>]}}})
    Index 2 Param 3=2
    Index 2 Param 4=0.7
    Module.Param0WriteEnabled=true
    Module.Param0WriteEnabled1=true
    Module.Param0WriteEnabled2=true
    Module.Param1WriteEnabled=true
    Module.Param1WriteEnabled1=true
    Module.Param1WriteEnabled2=true
    Module.Param2WriteEnabled=true
    Module.Param2WriteEnabled1=true
    Module.Param2WriteEnabled2=true
    Module.Param3WriteEnabled=false
    Module.Param3WriteEnabled1=true
    Module.Param3WriteEnabled2=true
    Write Parameter Index 1=true
    Write Parameter Index 2=true
   ) {module=DynamicMaterialParameters001}
   &CurlNoiseForce(
    Noise Frequency=250
    Noise Strength=&Multiply_Float(
     A=&FloatFromCurve(
      FloatCurve=DataInterface() {propertyValues {Curve {keys [<at 0 1>,<at 1 0>]}}}
      Scale Curve=&Lerp_Float(A=150, B=1000, Alpha=$Particles.InterpolateOverTime001.InterpolateOverTime_Float.Moving Average)
     )
     B=1
    )
    Pan Noise=false
   )
   &SolveForcesAndVelocity(Clamp Velocity=false, Limit Acceleration=false)
  ]
  rendererClasses [&NiagaraSpriteRendererProperties]
  renderers [
   {rendererClass=&NiagaraSpriteRendererProperties}
   {rendererClass=&NiagaraSpriteRendererProperties, rendererIndex=1}
  ]
 }}
summary {
 systemName=NS_Backpack_RebelClaw_Random_Smoke
 emitters [{emitterName=SmokePuff_Gen5, rendererClasses [&NiagaraSpriteRendererProperties]}]
 userVariables [
  User.bIsFrontEnd {type=NiagaraBool}
  User.bIsFrontEndPreview {type=NiagaraBool}
  User.EndColor {type=LinearColor, description="", defaultValue {struct=LinearColor, value=<rgba 0.0823 0.0823 0.0823 1>}}
  User.FadeOut {type=NiagaraFloat, defaultValue=0, description=""}
  User.Interpolate A {type=NiagaraFloat, defaultValue=20, description=""}
  User.Interpolate B {type=NiagaraFloat, defaultValue=50, description=""}
  User.SpawnScale {type=NiagaraFloat, defaultValue=1, description=""}
  User.StartColor {type=LinearColor, description="", defaultValue {struct=LinearColor, value=<rgba 0.117 0.1746 0.1812 1>}}
 ]
}
