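# Niagara smoke effect for the RebelClaw backpack cosmetic, puffing sprite smoke from two offset points on its skeletal mesh.
# SmokePuff_Gen5 spawns by rate and distance moved, partitions particles between the points, and inherits source movement as curl noise, drag and upward gravity shape the plume.
# Tint uses time of day and a StartColor-to-EndColor user lerp, an emitter-age curve envelopes alpha, and dynamic parameters animate the material.
# The WispySmoke mesh emitter is disabled.
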NiagaraSystem /BRCosmetics/Accessories/FORT_Backpacks/Backpack_F_MED_RebelClaw/FX/NS_Backpack_RebelClaw_Smoke
dependencies=null
emitters {SmokePuff_Gen5 {
  emitterUpdateScript [
   &EmitterState(Life Cycle Mode=::NewEnumerator1, Loop Behavior=::NewEnumerator1, Loop Delay=0.05, Loop Duration=1.5, UseLoopDelay=true)
   &SpawnPerUnit(Module.Use Max Movement Threshold=true, Module.Use Movement Tolerance=false, Use Spawn Probability=false)
   &SpawnRate(SpawnRate=8, Use Spawn Probability=false)
  ]
  particleSpawnScript [
   &PartitionParticles()
   &InitializeParticle(
    Color=&FNTimeOfDayColor(
     Night Color=<rgba 1.25 1.25 1.25 0.65>
     Morning Color=<rgba 1 1 1 0.65>
     Day Color=<rgba 1 1 1 0.65>
     Evening Color=<rgba 1.25 1.25 1.25 0.65>
    )
    Lifetime=&RandomRangeFloat(Minimum=0.7, Maximum=0.8)
    Sprite Rotation Mode=::NewEnumerator2
    Sprite Size Mode=::NewEnumerator3
    Uniform Sprite Size=30
    UsePositionOffset=false
    Write Material Random=true
   ) {module=InitializeParticle002}
   &AddVelocity(Velocity=<0 -25 10>)
   &SkeletalMeshLocation(
    Bone Sampling Mode=::NewEnumerator4
    MaskSpawnGroup=false
    Sampled Position Offset=&SelectVectorFromArray(
     Direct Array Index=$Particles.PartitionParticles.Partition
     Vector Selection Array=DataInterface() {propertyValues {FloatData [<-6 -15 12>, <5 -15 12>]}}
    )
    Skeletal Mesh=$System.Skeletal Mesh
   )
   &CameraOffset(Camera Offset Amount=&FN_FrontEndFloatModifier(Gameplay=12, Front End=0))
  ]
  particleUpdateScript [
   &ParticleState()
   &InheritSourceMovement(
    Applied Position Scale=&Multiply_VectorByFloat(
     Vector=<1 1 1>
     Float=&FloatFromCurve(FloatCurve=DataInterface() {propertyValues {Curve {keys [
          <at 0 1 RCIM_Cubic>
          {arriveTangent=-0.2, interpMode=RCIM_Cubic, leaveTangent=-0.2, time=0.25, value=1}
          {arriveTangent=-0.0269, interpMode=RCIM_Cubic, leaveTangent=-0.0269, tangentMode=RCTM_User, time=1, value=0.8}
         ]}}})
    )
    ApplyForce=false
    ApplyPosition=true
    ApplyPositionFalloff=false
   )
   &ScaleSpriteSize(Scale Factor=&Vector2DFromFloat(Value=&FloatFromCurve(
      FloatCurve=DataInterface() {propertyValues {Curve {keys [
          {arriveTangent=-0.0943, interpMode=RCIM_Cubic, leaveTangent=-0.0943, tangentMode=RCTM_User, time=0, value=0.3}
          {arriveTangent=0.116, interpMode=RCIM_Cubic, leaveTangent=0.116, tangentMode=RCTM_User, time=1, value=1}
         ]}}}
      Scale Curve=0.8
     )))
   &ScaleColor(
    Scale Alpha=&FloatFromCurve(
     FloatCurve=DataInterface() {propertyValues {Curve {keys [<at 0 0 RCIM_Cubic>,<at 0.2 0.5 RCIM_Cubic>,<at 1 0 RCIM_Cubic>]}}}
     Scale Curve=&Multiply_Float(
      A=3
      B=&FloatFromCurve(
       FloatCurve=DataInterface() {propertyValues {Curve {keys [
           <at 0 0 RCIM_Cubic>
           {arriveTangent=0.5, leaveTangent=0.5, time=0.5, value=1}
           {arriveTangent=-0.4, leaveTangent=-0.4, time=2, value=1}
           {interpMode=RCIM_Cubic, tangentMode=RCTM_User, time=3, value=0}
          ]}}}
       CurveIndex=$Emitter.Age
      )
     )
    )
    Scale RGB=&MakeVectorFromLinearColorRGB(Color=&LerpLinearColors(
      StartColor=$User.StartColor
      EndColor=$User.EndColor
      LerpFactor=&FloatFromCurve(FloatCurve=DataInterface() {propertyValues {Curve {keys [
           {arriveTangent=-0.1505, interpMode=RCIM_Cubic, leaveTangent=-0.1505, tangentMode=RCTM_User, time=0, value=0}
           {arriveTangent=-0.1542, interpMode=RCIM_Cubic, leaveTangent=-0.1542, tangentMode=RCTM_User, time=0.7, value=1}
          ]}}})
     ))
    ScaleA=true
    ScaleRGB=true
   ) {module=ScaleColor001}
   &DynamicMaterialParameters(
    Index 0 Param 1=&FloatFromCurve(
     FloatCurve=DataInterface() {propertyValues {Curve {keys [
         <at 0 0 RCIM_Cubic>
         {arriveTangent=1.5947, interpMode=RCIM_Cubic, leaveTangent=1.5947, tangentMode=RCTM_User, time=0.2557, value=0.5257}
         <at 1 1 RCIM_Cubic>
        ]}}}
     CurveIndex=$Particles.Age
     Scale Curve=0.25
    )
    Index 1 Param 1=1
    Index 1 Param 3=1
    Index 1 Param 4=0.5
    Index 2 Param 1=&FloatFromCurve(FloatCurve=DataInterface() {propertyValues {Curve {keys [
         {arriveTangent=1.9151, interpMode=RCIM_Cubic, leaveTangent=1.9151, tangentMode=RCTM_User, time=0, value=0}
         {interpMode=RCIM_Cubic, tangentMode=RCTM_User, time=1, value=1}
        ]}}})
    Index 2 Param 2=&FloatFromCurve(FloatCurve=DataInterface() {propertyValues {Curve {keys [<at 0 0.25 RCIM_Cubic>,<at 1 0 RCIM_Cubic>]}}})
    Index 2 Param 3=2
    Index 2 Param 4=0.7
    Module.Param0WriteEnabled=true
    Module.Param0WriteEnabled1=true
    Module.Param0WriteEnabled2=true
    Module.Param1WriteEnabled=true
    Module.Param1WriteEnabled1=true
    Module.Param1WriteEnabled2=true
    Module.Param2WriteEnabled=true
    Module.Param2WriteEnabled1=true
    Module.Param2WriteEnabled2=true
    Module.Param3WriteEnabled=false
    Module.Param3WriteEnabled1=true
    Module.Param3WriteEnabled2=true
    Write Parameter Index 1=true
    Write Parameter Index 2=true
   ) {module=DynamicMaterialParameters001}
   &CurlNoiseForce(
    Noise Frequency=250
    Noise Strength=&Multiply_Float(
     A=&FloatFromCurve(FloatCurve=DataInterface() {propertyValues {Curve {keys [<at 0 1>,<at 1 0>]}}}, Scale Curve=50)
     B=1
    )
    Pan Noise=false
   )
   &GravityForce(Gravity=&Multiply_VectorByFloat(
     Vector=<0 0 250>
     Float=&FloatFromCurve(FloatCurve=DataInterface() {propertyValues {Curve {keys [
          {interpMode=RCIM_Cubic, tangentMode=RCTM_User, time=0, value=0}
          {arriveTangent=2.3053, interpMode=RCIM_Cubic, leaveTangent=2.3053, tangentMode=RCTM_User, time=1, value=1}
         ]}}})
    ))
   &Drag(Drag=0.2, Use Linear Drag=true, Use Rotational Drag=true)
   &SolveForcesAndVelocity(Clamp Velocity=false, Limit Acceleration=false)
  ]
  rendererClasses [&NiagaraSpriteRendererProperties]
  renderers [
   {rendererClass=&NiagaraSpriteRendererProperties}
   {rendererClass=&NiagaraSpriteRendererProperties, rendererIndex=1}
  ]
 }}
summary {
 systemName=NS_Backpack_RebelClaw_Smoke
 emitters [
  {emitterName=SmokePuff_Gen5, rendererClasses [&NiagaraSpriteRendererProperties]}
  {bEnabled=false, emitterName=WispySmoke, rendererClasses [&NiagaraMeshRendererProperties]}
 ]
 userVariables [
  User.bIsFrontEnd {type=NiagaraBool}
  User.bIsFrontEndPreview {type=NiagaraBool}
  User.Color_1 {type=LinearColor, description="", defaultValue {struct=LinearColor, value=<rgba 0.2219 0.2292 0.2193 1>}}
  User.Color_2 {type=LinearColor, description="", defaultValue {struct=LinearColor, value=<rgba 0.1302 0.1266 0.1196 1>}}
  User.EndColor {type=LinearColor, description="", defaultValue {struct=LinearColor, value=<rgba 0.0823 0.0823 0.0823 1>}}
  User.FadeOut {type=NiagaraFloat, defaultValue=0, description=""}
  User.StartColor {type=LinearColor, description="", defaultValue {struct=LinearColor, value=<rgba 0.117 0.1746 0.1812 1>}}
 ]
}
