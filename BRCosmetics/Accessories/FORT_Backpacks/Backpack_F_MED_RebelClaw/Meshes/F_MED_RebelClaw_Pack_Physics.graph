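# Physics asset for the RebelClaw backpack cosmetic, providing bodies and joints for its left and right bullet and pipe pieces.
# Each piece pairs a small capsule on a base bone with a larger, offset capsule on a dynamic bone.
# Four constraints join each dynamic bone to its base, with a narrow swing limit on one axis and zero-degree limits on the other swing and twist, suggesting slight secondary jiggle.
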PhysicsAsset /BRCosmetics/Accessories/FORT_Backpacks/Backpack_F_MED_RebelClaw/Meshes/F_MED_RebelClaw_Pack_Physics
bodies {
 base_bullets_l [{shapeType=Capsule, length=0.505, radius=0.505, shapeName=base_bullets_l_capsule, rotation=<rot 0 -90 0>}]
 base_bullets_r [{shapeType=Capsule, length=0.505, radius=0.505, shapeName=base_bullets_r_capsule, rotation=<rot 0 -90 0>}]
 base_pipe_l [{shapeType=Capsule, length=0.505, radius=0.505, shapeName=base_pipe_l_capsule, rotation=<rot 0 -90 0>}]
 base_pipe_r [{shapeType=Capsule, length=0.505, radius=0.505, shapeName=base_pipe_r_capsule, rotation=<rot 0 -90 0>}]
 dynamic_bullets_l [{shapeType=Capsule, length=2, radius=1, shapeName=dynamic_bullets_l_capsule, center=<3 -3 0>, rotation=<rot 90 0 0>}]
 dynamic_bullets_r [{shapeType=Capsule, length=2, radius=1, shapeName=dynamic_bullets_r_capsule, center=<3 0 3>, rotation=<rot 90 0 0>}]
 dynamic_pipe_l [{shapeType=Capsule, length=2, radius=1, shapeName=dynamic_pipe_l_capsule, center=<2 -2 -2>, rotation=<rot 90 0 0>}]
 dynamic_pipe_r [{shapeType=Capsule, length=2, radius=1, shapeName=dynamic_pipe_r_capsule, center=<2 2 2>, rotation=<rot 90 0 0>}]
}
constraints [
 {bone1Name=dynamic_bullets_r, bone2Name=base_bullets_r, swing1LimitDegrees=0, swing2LimitDegrees=3, twistLimitDegrees=0}
 {bone1Name=dynamic_bullets_l, bone2Name=base_bullets_l, swing1LimitDegrees=2.5, swing2LimitDegrees=0, twistLimitDegrees=0}
 {bone1Name=dynamic_pipe_l, bone2Name=base_pipe_l, swing1LimitDegrees=6, swing2LimitDegrees=0, twistLimitDegrees=0}
 {bone1Name=dynamic_pipe_r, bone2Name=base_pipe_r, swing1LimitDegrees=6, swing2LimitDegrees=0, twistLimitDegrees=0}
]
